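; Quest interactable giant pink teddy bear that players carry, drop, throw or place, reporting the 2D distance carried in meters as quest progress.
; BeginPlay hides it unless a balance curve enables it; placing it on an overlapping player knocks it backward.
; A timer while held checks the instigator's location tags; leaving the POI triggers a client pop effect and server destruction.
; Destruction calls its spawner to spawn an actor.
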(event OnHeldObjectPlaced(FortHeldObject)
  (bind _pickuplocation (Variables|Default|GetPickupLocation))
  (bind _returnvalue (Math|Conversions|ToVector2D(Vector) _pickuplocation))
  (bind _fortheldobject (Variables|Default|GetFortHeldObject))
  (bind _returnvalue_2 (HeldObjects|Instigator|GetInstigator _fortheldobject))
  (bind _returnvalue_3 (Transformation|GetActorLocation _returnvalue_2))
  (bind _returnvalue_4 (Math|Conversions|ToVector2D(Vector) _returnvalue_3))
  (bind _returnvalue_5 (Math|Vector2D|Distance2D _returnvalue_4 _returnvalue))
  (bind _returnvalue_6 (/ _returnvalue_5 1e+02))
  (bind _returnvalue_7 (Math|Float|Floor _returnvalue_6))
  ;[1 1 1] When we set the bear down, if it is overlapping with a player, push the bear away from the direction the player is facing
  (Utilities|Array|ForEachLoopwithBreak (Collision|GetOverlappingActors (Variables|Default|GetCapsule) "/Script/FortniteGame.FortPlayerPawn" :!class /Script/Engine.PrimitiveComponent) :!label _label
    (:LoopBody
      (Variables|/Script/Engine.MovementComponent|SetVelocity (* (+ (* (Math|Vector|GetForwardVector (Transformation|GetActorRotation (HeldObjects|Instigator|GetInstigator (Variables|Default|GetFortHeldObject)))) -1.0) "0, 0,6.000000") 2e+02) (Variables|Default|GetFortProjectileMovement))
      (goto _label Break))
    (:Completed
      ;[1 1 1] Set variables to track player interaction; report quest progress
      (Variables|Default|SetHeld false)
      (if (> _returnvalue_7 0)
        (Quest|SendComplexCustomStatEvent (Quests|GetQuestManager (Fortnite|GetFortPlayerControllerfromActor (HeldObjects|Instigator|GetInstigator (Variables|Default|GetFortHeldObject))) "Invalid") 0 "(GameplayTags=)" (Variables|Default|GetQuestTargetTag) _returnvalue_7)
        (CallFunction|POIOverlapCheck)))))

(event OnHeldObjectDropped(FortHeldObject)
  (bind _pickuplocation (Variables|Default|GetPickupLocation))
  (bind _returnvalue (Math|Conversions|ToVector2D(Vector) _pickuplocation))
  (bind _fortheldobject (Variables|Default|GetFortHeldObject))
  (bind _returnvalue_2 (HeldObjects|Instigator|GetInstigator _fortheldobject))
  (bind _returnvalue_3 (Transformation|GetActorLocation _returnvalue_2))
  (bind _returnvalue_4 (Math|Conversions|ToVector2D(Vector) _returnvalue_3))
  (bind _returnvalue_5 (Math|Vector2D|Distance2D _returnvalue_4 _returnvalue))
  (bind _returnvalue_6 (/ _returnvalue_5 1e+02))
  (bind _returnvalue_7 (Math|Float|Floor _returnvalue_6))
  ;[1 1 1] Set variables to track player interaction; report quest progress
  (Variables|Default|SetHeld false)
  (if (> _returnvalue_7 0)
    (Quest|SendComplexCustomStatEvent (Quests|GetQuestManager (Fortnite|GetFortPlayerControllerfromActor (HeldObjects|Instigator|GetInstigator (Variables|Default|GetFortHeldObject))) "Invalid") 0 "(GameplayTags=)" (Variables|Default|GetQuestTargetTag) _returnvalue_7)
    (CallFunction|POIOverlapCheck)))

(event OnHeldObjectThrown(FortHeldObject)
  (bind _pickuplocation (Variables|Default|GetPickupLocation))
  (bind _returnvalue (Math|Conversions|ToVector2D(Vector) _pickuplocation))
  (bind _fortheldobject (Variables|Default|GetFortHeldObject))
  (bind _returnvalue_2 (HeldObjects|Instigator|GetInstigator _fortheldobject))
  (bind _returnvalue_3 (Transformation|GetActorLocation _returnvalue_2))
  (bind _returnvalue_4 (Math|Conversions|ToVector2D(Vector) _returnvalue_3))
  (bind _returnvalue_5 (Math|Vector2D|Distance2D _returnvalue_4 _returnvalue))
  (bind _returnvalue_6 (/ _returnvalue_5 1e+02))
  (bind _returnvalue_7 (Math|Float|Floor _returnvalue_6))
  ;[1 1 1] Set variables to track player interaction; report quest progress
  (Variables|Default|SetHeld false)
  (if (> _returnvalue_7 0)
    (Quest|SendComplexCustomStatEvent (Quests|GetQuestManager (Fortnite|GetFortPlayerControllerfromActor (HeldObjects|Instigator|GetInstigator (Variables|Default|GetFortHeldObject))) "Invalid") 0 "(GameplayTags=)" (Variables|Default|GetQuestTargetTag) _returnvalue_7)
    (CallFunction|POIOverlapCheck)))

(event EventBeginPlay
  ;[1 1 1] Begin Play
  (|CallParentFunction)
  (Collision|SetCollisionResponsetoChannel (Variables|Default|GetStaticMesh) "ECC_Camera" "ECR_Ignore")
  (if (not (CallFunction|ShouldBearbeEnabled))
    (Rendering|SetActorHiddenInGame true)
    (Collision|SetActorEnableCollision false)))

(event OnHeldObjectPickedUp(FortHeldObject)
  ;[1 1 1] Set variables to track player interaction; report quest progress
  (Variables|Default|SetHeld true)
  (Variables|Default|SetPickupLocation (Transformation|GetActorLocation (HeldObjects|Instigator|GetInstigator (Variables|Default|GetFortHeldObject)))))

(event Interaction|EventBlueprintOnInteract
  (params (InteractingPawn :type Fort%20Pawn%20Object%20Reference) (InteractionBeingAttempted :type EInteractionBeingAttempted%20Enum :default "FirstInteraction"))
  ;[1 1 1] On Interact
  (switch Utilities|FlowControl|Switch|SwitchonEInteractionBeingAttempted InteractionBeingAttempted
    (:SecondInteraction
      (bind _asfort_player_pawn (Utilities|Casting|CastToFortPlayerPawn InteractingPawn))
      (CallFunction|SetInstigator _asfort_player_pawn)
      (Variables|/Script/Engine.MovementComponent|SetVelocity "0, 0, 0" (Variables|Default|GetFortProjectileMovement))
      (if (Networking|IsServer)
        (HeldObjects|Behavior|PickupHeldObject (Variables|Default|GetFortHeldObject) _asfort_player_pawn)))))

(event Custom|SetInstigator
  (params (Instigator :type Fort%20Player%20Pawn%20Object%20Reference))
  ;[1 1 1] On Interact
  (bind _output_get (Variables|Actor|SetInstigator Instigator)))

(event Custom|POI%20Overlap%20Check
  (if (Networking|IsServer)
    (bind _returnvalue (GameplayTags|HasAnyMatchingGameplayTags(Message) (stale Custom|SetInstigator _output_get) (Variables|Default|GetLocationTags)))
    (Variables|Default|SetbIsInPOI _returnvalue)))

(event Custom|ClientEnd
  (if (Networking|IsServer)
    (Utilities|FlowControl|Delay 0.1)
    (Actor|DestroyActor)
    (else
      (Components|Activation|Activate (Variables|Default|GetDeathExplosion) false)
      (Development|SetHiddeninGame (Variables|Default|GetStaticMesh) true false))))

(event OnLinkedActorDisconnected(FortLinkToActor)
  (if (not (Variables|Default|GetHeld))
    (Variables|ProjectileMovement|SetRepStartMoving 1)))

(event EventDestroyed
  (Class|BActorSpawnerParent|SpawnActor :self (Variables|Default|GetMySpawner)))

(fn BlueprintCanInteract
  (params (InteractingPawn :type Fort%20Pawn%20Object%20Reference) (InteractionBeingAttempted :type EInteractionBeingAttempted%20Enum :default "FirstInteraction") (InteractionType :type TInteractionType%20Enum :default "IT_NoInteraction"))
  (Utilities|IsValid InteractingPawn
    (:Is%20Valid
      (bind _returnvalue (HeldObjects|Interaction|CanInteract (Variables|Default|GetFortHeldObject) (Utilities|Casting|CastToFortPlayerPawn InteractingPawn)))
      (return (Math|Boolean|ANDBoolean _returnvalue (not (GameplayTags|HasMatchingGameplayTag InteractingPawn "(TagName=\"Gameplay.Status.HidingInProp\")")) (not (GameplayTags|HasAnyMatchingGameplayTags InteractingPawn (Variables|Interact|GetTC_BlockPickup))))))))

(fn BlueprintGetInteractionString
  (params (InteractingPawn :type Fort%20Pawn%20Object%20Reference) (InteractionBeingAttempted :type EInteractionBeingAttempted%20Enum :default "FirstInteraction"))
  (return (Variables|Default|GetInteractString)))

(fn OnRep_bIsInPOI
  (if (not (Variables|Default|GetbIsInPOI))
    (CallFunction|ClientEnd)))

(fn OnRep_Held
  (if (Variables|Default|GetHeld)
    (Utilities|Time|SetTimerbyFunctionName 0 "POI Overlap Check" 0.1 true)
    (else
      (Utilities|Time|ClearTimerbyFunctionName 0 "POI Overlap Check"))))

(fn ShouldBearBeEnabled
  (bind _outxy (DataTable|EvaluateCurveTableRow "/Game/Athena/Balance/DataTables/AthenaGameData.AthenaGameData" "Default.QuestInteractable.GiantPinkBear.Enabled")
    (:RowFound
      (if (>= _outxy 1.0)
        (return true)
        (else
          (return false))))
    (:RowNotFound
      (return false))))

(fn ConstructionScript
  (|CallParentFunction)
  (Effects|Components|ParticleSystem|SetTemplate (Variables|Default|GetDeathExplosion) (Variables|Default|GetBearPopEffect)))
