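; Actor spawner for the Big Pink Teddy held-object quest interactable, used to place the bear in the world.
; On its spawn event, it spawns the configured class at its own transform if that class is valid.
; A spawned teddy bear receives the spawner's POI list as location tags and a reference back to the spawner.
; Any other result is cast to BuildingActor and initialized as a kismet-spawned building actor.
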(event EventBeginPlay
  (|CallParentFunction) ; This node is disabled and will not be called.\nDrag off pins to build functionality.
)

(event Collision|EventActorBeginOverlap
  (params (OtherActor :type Actor%20Object%20Reference)))

(event EventTick
  (params (DeltaSeconds :type Float%20%28single-precision%29)))

(event EventSpawnActor
  (params (New%20Transform :type Transform))
  (bind _classtospawn (Variables|SpawnerVars|GetClassToSpawn))
  ;[1 1 1] If the Actor is a valid class spawn it. If it is a building actor init the kismet of it.
  (if (Utilities|IsValidClass _classtospawn)
    (bind _returnvalue (Game|SpawnActorfromClass _classtospawn (Transformation|GetActorTransform) "AlwaysSpawn" "SelectDefaultAtRuntime" 0 0))
    (bind _asbga_held_object_quest_big_pink_teddy_bear (Utilities|Casting|CastToBGA_HeldObject_Quest_BigPinkTeddyBear _returnvalue)
      (:then
        (Variables|/Game/Athena/Items/QuestInteractables/BigPinkTeddyBear/BGA_HeldObject_Quest_BigPinkTeddyBear.BGA_HeldObject_Quest_BigPinkTeddyBear_C|SetLocationTags (Variables|Default|GetPOI%20List) _asbga_held_object_quest_big_pink_teddy_bear)
        (Variables|/Game/Athena/Items/QuestInteractables/BigPinkTeddyBear/BGA_HeldObject_Quest_BigPinkTeddyBear.BGA_HeldObject_Quest_BigPinkTeddyBear_C|SetMySpawner self _asbga_held_object_quest_big_pink_teddy_bear))
      (:CastFailed
        (bind _asbuilding_actor (Utilities|Casting|CastToBuildingActor _returnvalue))
        (Building|InitializeKismetSpawnedBuildingActor _asbuilding_actor 0 0 false 0 false)))))

(fn ConstructionScript
  (|CallParentFunction))
